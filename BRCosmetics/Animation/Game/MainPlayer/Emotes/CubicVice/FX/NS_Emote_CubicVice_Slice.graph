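# Niagara system for a slice effect in the CubicVice emote, built from burst-spawned, short-lived sprites.
# MainSlash draws a bright, thin slash, DelayedSlash follows with a dimmer one, and Flare adds a quick pulse at the start.
# Directional_Streaks move along a system direction vector.
# Slash and streak emitters layer extra sprite renderers with separate size and color attributes for darker layers, and time-of-day scalars adjust most brightness.
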NiagaraSystem /BRCosmetics/Animation/Game/MainPlayer/Emotes/CubicVice/FX/NS_Emote_CubicVice_Slice
dependencies {
 usedDataInterfaces [
  &NiagaraDataInterfaceColorCurve
  &NiagaraDataInterfaceCurve
  &NiagaraDataInterfaceVector2DCurve
  &NiagaraDataInterfaceSkeletalMesh
 ]
 usedDynamicInputs [&Vector4FromCurve]
 usedModules [
  &self:SystemSpawnScript.SystemScriptSource.SystemScriptGraph.NiagaraNodeAssignment_0.SetVariables_#1
  &self:SystemSpawnScript.SystemScriptSource.SystemScriptGraph.NiagaraNodeAssignment_1.SetVariables_#2
  &SystemState
  /BRCosmetics/Characters/Player/Female/Medium/Bodies/F_MED_PiperShelf_Pearl/FX/TimeOfDayGlobal
 ]
}
emitters {
 DelayedSlash {
  emitterUpdateScript [
   &EmitterState(Life Cycle Mode=::NewEnumerator1, Loop Behavior=::NewEnumerator1, Loop Duration=1, UseLoopDelay=false)
   &SpawnBurst_Instantaneous(Spawn Count=1, Spawn Time=0.2, Use Spawn Probability=true)
  ]
  particleSpawnScript [
   &InitializeParticle(
    Color Mode=::NewEnumerator0
    Lifetime=0.4
    Position Offset=$System.StartOffset
    Position Offset Coordinate Space=::Local
    Sprite Size=<3.5 150>
    Sprite Size Mode=::NewEnumerator4
    UsePositionOffset=true
    Write Material Random=true
   ) {module=InitializeParticle001}
   &CameraOffset(Camera Offset Amount=20)
   &DynamicMaterialParameters(
    Index 0 Param 1=0.2
    Index 0 Param 2=1
    Module.Param0WriteEnabled=true
    Module.Param1WriteEnabled=true
    Module.Param2WriteEnabled=true
    Module.Param3WriteEnabled=true
   )
   &SpriteFacingAndAlignment(Alignment Coordinate Space=::Local, Sprite Alignment=<1 0 -0.2>, Sprite Facing=<0 0 1>, WriteAlignment=true, WriteFacing=true)
  ]
  particleUpdateScript [
   &ParticleState()
   &SolveForcesAndVelocity(Clamp Velocity=false, Limit Acceleration=false)
   &ScaleSpriteSize(
    Non-Uniform Curve Sprite Scale=DataInterface() {propertyValues {
      XCurve {keys [
        <at 0 0 RCIM_Cubic>
        {arriveTangent=0.797, interpMode=RCIM_Cubic, leaveTangent=0.797, tangentMode=RCTM_User, time=0.4, value=0.9}
        {arriveTangent=-0.3607, interpMode=RCIM_Cubic, leaveTangent=-0.3607, tangentMode=RCTM_User, time=0.9, value=0.85}
        <at 1 -0.1 RCIM_Cubic>
       ]}
      YCurve {keys [
        <at 0 0 RCIM_Cubic>
        {arriveTangent=0.0607, interpMode=RCIM_Cubic, leaveTangent=0.0607, tangentMode=RCTM_User, time=0.4, value=0.9}
        {arriveTangent=-1.5833, interpMode=RCIM_Cubic, leaveTangent=-1.5833, time=0.8, value=0.85}
        <at 1 -0.05 RCIM_Cubic>
       ]}
     }}
    Scale Sprite Size Mode=::NewEnumerator3
   ) {module=ScaleSpriteSize001}
   &ScaleColor(
    Scale Alpha=&FloatFromCurve(FloatCurve=DataInterface() {propertyValues {Curve {keys [<at 0.75 1.05>,<at 1 0>]}}})
    Scale RGB=&MakeVectorFromLinearColorRGB(Color=&ColorFromCurve(ColorCurve=DataInterface() {propertyValues {
        AlphaCurve {keys [<at 0 1 RCIM_Cubic>,<at 0.5792 1>]}
        BlueCurve {keys [
          <at 0.2599 0>
          <at 0.4789 0.0365>
          <at 0.624 0.0781>
          {arriveTangent=0.4987, interpMode=RCIM_Cubic, leaveTangent=0.4987, time=0.8404, value=0.7344}
          <at 1 0.2656 RCIM_Cubic>
         ]}
        GreenCurve {keys [
          <at 0.2599 0>
          <at 0.4789 0.0365>
          <at 0.624 0.0781>
          {arriveTangent=0.4987, interpMode=RCIM_Cubic, leaveTangent=0.4987, time=0.8404, value=0.7344}
          <at 1 0.2656 RCIM_Cubic>
         ]}
        RedCurve {keys [
          <at 0.2599 0>
          <at 0.4789 0.0365>
          <at 0.624 0.0781>
          {arriveTangent=0.4987, interpMode=RCIM_Cubic, leaveTangent=0.4987, time=0.8404, value=0.7344}
          <at 1 0.2656 RCIM_Cubic>
         ]}
       }}))
    ScaleA=true
    ScaleRGB=true
   )
   &FN_Sprite_Attributes(
    Color=&MakeLinearColorFromVectorAndFloat(
     "Vector (RGB)"=&MakeVectorFromLinearColorRGB(Color=<rgba 0.0677 0.0677 0.0677 1>)
     "Float (Alpha)"=&FloatFromCurve(FloatCurve=DataInterface() {propertyValues {Curve {keys [
          {interpMode=RCIM_Cubic, tangentMode=RCTM_User, time=0, value=0}
          {arriveTangent=0.0399, interpMode=RCIM_Cubic, leaveTangent=0.0399, tangentMode=RCTM_User, time=0.3, value=0.925}
          {interpMode=RCIM_Cubic, tangentMode=RCTM_User, time=0.8, value=0.95}
          {tangentMode=RCTM_User, time=1, value=0}
         ]}}})
    )
    Create Color Attribute=true
    Create SpriteSize Attribute=true
    SpriteSize=&Multiply_Vector2D(A=$Particles.Initial.SpriteSize, B=<2.5 1.5>)
    Write to Module Namespace=true
   )
  ]
  rendererClasses [&NiagaraSpriteRendererProperties]
  renderers [
   {rendererClass=&NiagaraSpriteRendererProperties}
   {rendererClass=&NiagaraSpriteRendererProperties, rendererIndex=1}
  ]
 }
 Directional_Streaks {
  emitterUpdateScript [
   &EmitterState()
   &SpawnBurst_Instantaneous(Spawn Time=0.02, Use Spawn Probability=true, UseLoopCountLimit=false)
  ]
  particleSpawnScript [
   &InitializeParticle(
    Lifetime=0.15
    Sprite Size=<4 70>
    Sprite Size Mode=::NewEnumerator4
    Sprite UV Mode=::NewEnumerator1
    UsePositionOffset=false
    Write Material Random=true
   ) {module=InitializeParticle001}
   &SubUVAnimation(End Frame=3, SubUV Animation Mode:ENiagara_SubUVLookupMode=::NewEnumerator2)
   &AddVelocity(Velocity=$System.DirectionVector, Velocity Speed Scale=3)
   &CameraOffset()
   /SetVariables(Particles.Size Blurry Streak:Vector2f=<20 150>) {isSetParameters=true}
   /SetVariables(Particles.Color Dark Streak:LinearColor=<rgba 0.2344 0.2344 0.2344 0.5>, Particles.Size Dark Streak:Vector2f=<5 100>) {isSetParameters=true}
   &SpriteFacingAndAlignment(Alignment Coordinate Space=::Local, Sprite Alignment=<1 0 -0.2>, Sprite Facing=<0 0 1>, WriteAlignment=true, WriteFacing=true)
  ]
  particleUpdateScript [
   &ParticleState()
   &SolveForcesAndVelocity(Clamp Velocity=false, Limit Acceleration=false)
   &ScaleSpriteSize(Scale Factor=&Vector2DFromCurve(
     Vector2Curve=DataInterface() {propertyValues {XCurve {keys [<at 0 1 RCIM_Cubic>,<at 1 0 RCIM_Cubic>]}, YCurve {keys [<at 0 0.5 RCIM_Cubic>,<at 1 2 RCIM_Cubic>]}}}
     ScaleCurve=&Multiply_Vector2D_ByFloat(Float=1, Vector2D=<1.5 1>)
    ))
   &ScaleColor(
    Scale Alpha=&FloatFromCurve(FloatCurve=DataInterface() {propertyValues {Curve {keys [<at 0 1 RCIM_Cubic>,<at 1 0>]}}}, Scale Curve=0.2)
    Scale RGB=&VectorFromFloat(Value=&FloatFromCurve(
      FloatCurve=DataInterface() {propertyValues {Curve {keys [<at 0 10 RCIM_Cubic>,<at 1 1 RCIM_Cubic>]}}}
      Scale Curve=&Multiply_Float(
       A=0.5
       B=&FNTimeOfDayScalar(Night Scale=0.25, Morning Scale=0.45, Day Scale=0.8, Evening Scale=0.45)
      )
     ))
    ScaleA=true
    ScaleRGB=true
   )
   /SetVariables(
    Particles.Color Dark Streak:LinearColor=&MakeLinearColorFromVectorAndFloat(
     "Vector (RGB)"=&MakeVectorFromLinearColorRGB(Color=$Particles.Initial.Color Dark Streak)
     "Float (Alpha)"=&FloatFromCurve(FloatCurve=DataInterface() {propertyValues {Curve {keys [{interpMode=RCIM_Cubic, tangentMode=RCTM_User, time=0, value=1}, {interpMode=RCIM_Cubic, tangentMode=RCTM_User, time=0.5, value=0}]}}})
    )
    Particles.Size Dark Streak:Vector2f=&Multiply_Vector2D(
     A=$Particles.Initial.Size Dark Streak
     B=&MakeVector2D(
      X=1
      Y=&FloatFromCurve(
       FloatCurve=DataInterface() {propertyValues {Curve {keys [
           {arriveTangent=1.3254, interpMode=RCIM_Cubic, leaveTangent=1.3254, tangentMode=RCTM_User, time=0, value=0}
           {interpMode=RCIM_Cubic, tangentMode=RCTM_User, time=1, value=1}
          ]}}}
       Scale Curve=0.5
      )
     )
    )
   ) {isSetParameters=true}
  ]
  rendererClasses [&NiagaraSpriteRendererProperties]
  renderers [
   {rendererClass=&NiagaraSpriteRendererProperties}
   {rendererClass=&NiagaraSpriteRendererProperties, rendererIndex=1}
   {rendererClass=&NiagaraSpriteRendererProperties, rendererIndex=2}
  ]
 }
 Flare {
  emitterUpdateScript [
   &EmitterState(Life Cycle Mode=::NewEnumerator1, Loop Behavior=::NewEnumerator1, Loop Duration=1, UseLoopDelay=false)
   &SpawnBurst_Instantaneous(Use Spawn Probability=false, UseLoopCountLimit=false)
  ]
  particleSpawnScript [
   &InitializeParticle(
    Color Mode=::NewEnumerator0
    Lifetime=0.1
    Position Offset=$System.StartOffset
    Sprite Size Mode=::NewEnumerator3
    Uniform Sprite Size=100
    UsePositionOffset=true
    Write Material Random=true
   )
   &CameraOffset(Camera Offset Amount=100)
  ]
  particleUpdateScript [
   &ParticleState()
   &ScaleSpriteSize(Uniform Curve Sprite Scale=DataInterface() {propertyValues {Curve {keys [
        {interpMode=RCIM_Cubic, tangentMode=RCTM_User, time=0, value=0.5}
        {interpMode=RCIM_Cubic, tangentMode=RCTM_User, time=0.5, value=1}
        {tangentMode=RCTM_User, time=1, value=0.5}
       ]}}})
   &ScaleColor(
    Scale Alpha=&FloatFromCurve(
     FloatCurve=DataInterface() {propertyValues {Curve {keys [
         {interpMode=RCIM_Cubic, tangentMode=RCTM_User, time=0, value=0}
         {interpMode=RCIM_Cubic, tangentMode=RCTM_User, time=0.5, value=1}
         {tangentMode=RCTM_User, time=1, value=0}
        ]}}}
     Scale Curve=0.1
    )
    Scale RGB=&Multiply_VectorByFloat(
     Vector=<1 1 1>
     Float=&FNTimeOfDayScalar(Night Scale=0.5, Morning Scale=0.6, Day Scale=0.8, Evening Scale=0.6)
    )
    ScaleA=true
    ScaleRGB=true
   )
  ]
  rendererClasses [&NiagaraSpriteRendererProperties]
  renderers [{rendererClass=&NiagaraSpriteRendererProperties}]
 }
 MainSlash {
  emitterUpdateScript [
   &EmitterState(Life Cycle Mode=::NewEnumerator1, Loop Behavior=::NewEnumerator1, Loop Duration=1, UseLoopDelay=false)
   &SpawnBurst_Instantaneous(Spawn Count=1, Use Spawn Probability=true)
  ]
  particleSpawnScript [
   &InitializeParticle(
    Color=<rgba 5 5 5 1>
    Lifetime=0.225
    Position Offset=$System.StartOffset
    Position Offset Coordinate Space=::Local
    Sprite Size=<7.5 200>
    Sprite Size Mode=::NewEnumerator4
    UsePositionOffset=true
    Write Material Random=true
   ) {module=InitializeParticle001}
   &CameraOffset(Camera Offset Amount=20)
   &DynamicMaterialParameters(
    Index 0 Param 1=0.2
    Index 0 Param 2=1
    Module.Param0WriteEnabled=true
    Module.Param1WriteEnabled=true
    Module.Param2WriteEnabled=true
    Module.Param3WriteEnabled=true
   )
   &SpriteFacingAndAlignment(Alignment Coordinate Space=::Local, Sprite Alignment=<1 0 -0.2>, Sprite Facing=<0 0 1>, WriteAlignment=true, WriteFacing=true)
  ]
  particleUpdateScript [
   &ParticleState()
   &SolveForcesAndVelocity(Clamp Velocity=false, Limit Acceleration=false)
   &ScaleColor(
    Color Value To Scale=$Particles.Color
    Scale Alpha=&FloatFromCurve(FloatCurve=DataInterface() {propertyValues {Curve {keys [{interpMode=RCIM_Cubic, tangentMode=RCTM_User, time=0, value=1}, {interpMode=RCIM_Cubic, tangentMode=RCTM_User, time=1, value=0}]}}})
    Scale RGB=&Multiply_VectorByFloat(
     Vector=&MakeVectorFromLinearColorRGB(Color=&ColorFromCurve(ColorCurve=DataInterface() {propertyValues {
         AlphaCurve {keys [<at 0 1 RCIM_Cubic>]}
         BlueCurve {keys [<at 0 0.5>,<at 0.2843 1 RCIM_Cubic>,<at 0.5346 0.5 RCIM_Cubic>]}
         GreenCurve {keys [<at 0 0.5>,<at 0.2843 1 RCIM_Cubic>,<at 0.5346 0.5 RCIM_Cubic>]}
         RedCurve {keys [<at 0 0.5>,<at 0.2843 1 RCIM_Cubic>,<at 0.5346 0.5 RCIM_Cubic>]}
        }}))
     Float=&FNTimeOfDayScalar(Night Scale=0.25, Morning Scale=0.35, Day Scale=0.7, Evening Scale=0.35)
    )
    ScaleA=true
    ScaleRGB=true
   )
   &ScaleSpriteSize(
    Non-Uniform Curve Sprite Scale=DataInterface() {propertyValues {
      XCurve {keys [<at 0 0 RCIM_Cubic>, {arriveTangent=-0.1, interpMode=RCIM_Cubic, leaveTangent=-0.1, time=0.2, value=1.075}, <at 1 -0.1 RCIM_Cubic>]}
      YCurve {keys [
        <at 0 0 RCIM_Cubic>
        {arriveTangent=0.0607, interpMode=RCIM_Cubic, leaveTangent=0.0607, tangentMode=RCTM_User, time=0.1, value=1.025}
        <at 1 0.8 RCIM_Cubic>
       ]}
     }}
    Scale Sprite Size Mode=::NewEnumerator3
   ) {module=ScaleSpriteSize001}
   &FN_Sprite_Attributes(
    Color=<rgba 0 0 0 1>
    Create Color Attribute=true
    Create SpriteSize Attribute=true
    SpriteSize=&Multiply_Vector2D(A=$Particles.SpriteSize, B=<1.1 1.25>)
    Write to Module Namespace=true
   )
  ]
  rendererClasses [&NiagaraSpriteRendererProperties]
  renderers [
   {rendererClass=&NiagaraSpriteRendererProperties}
   {rendererClass=&NiagaraSpriteRendererProperties, rendererIndex=1}
  ]
 }
}
summary {
 systemName=NS_Emote_CubicVice_Slice
 emitters [
  {emitterName=MainSlash, rendererClasses [&NiagaraSpriteRendererProperties]}
  {emitterName=Directional_Streaks, rendererClasses [&NiagaraSpriteRendererProperties]}
  {emitterName=Flare, rendererClasses [&NiagaraSpriteRendererProperties]}
  {emitterName=DelayedSlash, rendererClasses [&NiagaraSpriteRendererProperties]}
 ]
 userVariables [
  User.bIsFrontEnd {type=NiagaraBool}
  User.bIsFrontEndPreview {type=NiagaraBool}
 ]
}
