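# One-shot Niagara water splat for the CubicVice emote, spawned from three bone or socket indices on a skeletal mesh.
# Three burst emitters fire together: stretched falling droplets, expanding splash sprites animated by dynamic material parameters, and brief windstreak jets, all inheriting source movement.
# A front-end modifier reduces droplet speed and size for menus.
# Time-of-day colour tints droplets and streaks, while streak opacity also follows quality level.
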NiagaraSystem /BRCosmetics/Animation/Game/MainPlayer/Emotes/CubicVice/FX/NS_Emote_CubicVice_Splat
dependencies {
 usedDataInterfaces [
  &NiagaraDataInterfaceArrayInt32
  &NiagaraDataInterfaceCurve
  &NiagaraDataInterfaceVector2DCurve
  &NiagaraDataInterfaceSpriteRendererInfo
  &NiagaraDataInterfaceArrayFloat
  &NiagaraDataInterfaceSkeletalMesh
  &NiagaraDataInterfaceColorCurve
 ]
 usedDynamicInputs [&MakeFloatFromVector2d, &Multiply_Vector2]
 usedModules [
  &self:SystemSpawnScript.SystemScriptSource.SystemScriptGraph.NiagaraNodeAssignment_0.SetVariables_#1
  &SystemState
  /BRCosmetics/Characters/Player/Female/Medium/Bodies/F_MED_PiperShelf_Pearl/FX/TimeOfDayGlobal
 ]
}
emitters {
 WaterDrops {
  emitterUpdateScript [
   &EmitterState()
   &SpawnBurst_Instantaneous(
    Spawn Count=&RandomRangeInt(Minimum=6, Maximum=10)
    Spawn Time=0.05
    Use Spawn Probability=false
    UseLoopCountLimit=false
   )
  ]
  particleSpawnScript [
   &InitializeParticle(
    Color=<rgba 1 1 1 0.75>
    Lifetime Max=0.6
    Lifetime Min=0.4
    Lifetime Mode=::NewEnumerator1
    Sprite Size=&Multiply_Vector2D_ByFloat(Float=&RandomRangeFloat(Minimum=1.5, Maximum=2), Vector2D=<0.25 1>)
    Sprite Size Mode=::NewEnumerator4
    UsePositionOffset=false
    Write Material Random=true
   )
   &AddVelocity(
    Module.UseConeFalloff=false
    Module.UseVelDistribution=false
    Override Seed=false
    Velocity Mode=::NewEnumerator2
    Velocity Speed=&Multiply_Float(
     A=&RandomRangeFloat(Minimum=200, Maximum=400)
     B=&FN_FrontEndFloatModifier(Gameplay=1, Front End=0.6)
    )
   )
   &SkeletalMeshLocation(
    Bone / Socket Index=&SelectIntFromArray(
     Array Sampling Mode=::NewEnumerator1
     Direct Array Index=?
     Int Selection Array=DataInterface() {propertyValues {IntData [0,1,2]}}
     Random Seed=0
    )
    Bone Sampling Mode=::NewEnumerator4
    MaskSpawnGroup=false
    Skeletal Mesh=$System.Skeletal Mesh
   )
  ]
  particleUpdateScript [
   &ParticleState()
   &ScaleSpriteSize(
    Non-Uniform Curve Scale=&Vector2DFromFloat(Value=&FN_FrontEndFloatModifier(Gameplay=1, Front End=0.6))
    Non-Uniform Curve Sprite Scale=DataInterface() {propertyValues {
      XCurve {keys [
        {arriveTangent=0.0516, interpMode=RCIM_Cubic, leaveTangent=0.0516, tangentMode=RCTM_User, time=0, value=2}
        {interpMode=RCIM_Cubic, tangentMode=RCTM_User, time=1, value=0.47}
       ]}
      YCurve {keys [
        {arriveTangent=-0.0866, interpMode=RCIM_Cubic, leaveTangent=-0.0866, tangentMode=RCTM_User, time=0, value=4}
        {interpMode=RCIM_Cubic, tangentMode=RCTM_User, time=0.9988, value=0.22}
       ]}
     }}
    Scale Sprite Size Mode=::NewEnumerator3
   )
   &GravityForce(Gravity=&Multiply_VectorByFloat(
     Vector=<0 0 -980>
     Float=&FloatFromCurve(FloatCurve=DataInterface() {propertyValues {Curve {keys [<at 0 0>,<at 0.2 0>,<at 0.25 1>,<at 1 0>]}}})
    ))
   &Drag(Drag=3, Use Linear Drag=true, Use Rotational Drag=true)
   &InheritSourceMovement(ApplyForce=false, ApplyPosition=true, ApplyPositionFalloff=false)
   &SolveForcesAndVelocity(Clamp Velocity=false, Limit Acceleration=false)
   &ScaleSpriteSizeBySpeed(Max Scale Factor=<1 3>)
   &ScaleColor(
    Scale Alpha=&FloatFromCurve(
     FloatCurve=DataInterface() {propertyValues {Curve {keys [<at 0.7 1 RCIM_Cubic>, {arriveTangent=-9.6037, interpMode=RCIM_Cubic, leaveTangent=-9.6037, tangentMode=RCTM_User, time=1, value=0}]}}}
     Scale Curve=&RandomRangeFloat(Minimum=0.75, Maximum=1)
    )
    Scale RGB=&Lerp_Vector(
     A=&MakeVectorFromLinearColorRGB(Color=<rgba 1 1 1 1>)
     B=&MakeVectorFromLinearColorRGB(Color=$System.TODColor)
     Alpha=0.5
    )
    ScaleA=true
    ScaleRGB=true
   )
  ]
  rendererClasses [&NiagaraSpriteRendererProperties]
  renderers [{rendererClass=&NiagaraSpriteRendererProperties}]
 }
 WaterSplash {
  emitterUpdateScript [
   &EmitterState()
   &SpawnBurst_Instantaneous(Spawn Count=3, Spawn Time=0.05, Use Spawn Probability=false, UseLoopCountLimit=false)
  ]
  particleSpawnScript [
   &InitializeParticle(
    Lifetime Max=0.3
    Lifetime Min=0.25
    Lifetime Mode=::NewEnumerator1
    Sprite Rotation Mode=::NewEnumerator2
    Sprite Size Mode=::NewEnumerator1
    Uniform Sprite Size Max=16
    Uniform Sprite Size Min=14
    UsePositionOffset=false
    Write Material Random=true
   )
   &SkeletalMeshLocation(
    Bone / Socket Index=&SelectIntFromArray(
     Direct Array Index=&ReturnExecIndex()
     Int Selection Array=DataInterface() {propertyValues {IntData [0,1,2]}}
    )
    Bone Sampling Mode=::NewEnumerator4
    MaskSpawnGroup=false
    Skeletal Mesh=$System.Skeletal Mesh
   )
   &AddVelocity(
    Cone Angle=25
    Module.UseConeFalloff=false
    Module.UseVelDistribution=false
    Override Seed=false
    Velocity Mode=::NewEnumerator2
    Velocity Speed=&Multiply_Float(A=&RandomRangeFloat(Minimum=150, Maximum=175), B=1)
   )
   &SubUVAnimation(End Frame=0, SubUV Animation Mode:ENiagara_SubUVLookupMode=::NewEnumerator2)
   &CameraOffset(Camera Offset Amount=2)
  ]
  particleUpdateScript [
   &ParticleState()
   &DynamicMaterialParameters(
    Index 0 Param 2=&FloatFromCurve(FloatCurve=DataInterface() {propertyValues {Curve {keys [<at 0.35 0.6 RCIM_Cubic>,<at 1 0.7 RCIM_Cubic>]}}})
    Index 0 Param 3=&FloatFromCurve(FloatCurve=DataInterface() {propertyValues {Curve {keys [<at 0.2 -0.5>,<at 0.9 -1>]}}})
    Index 0 Param 4=&FloatFromCurve(FloatCurve=DataInterface() {propertyValues {Curve {keys [{arriveTangent=2.6651, interpMode=RCIM_Cubic, leaveTangent=2.6651, tangentMode=RCTM_User, time=0, value=0.1}, <at 0.6 1.333 RCIM_Cubic>]}}})
    Module.Param0WriteEnabled=true
    Module.Param1WriteEnabled=true
    Module.Param2WriteEnabled=true
    Module.Param3WriteEnabled=true
   )
   &GravityForce(Gravity=<0 0 -300>)
   &Drag(Drag=2, Use Linear Drag=true, Use Rotational Drag=true)
   &InheritSourceMovement(ApplyForce=false, ApplyPosition=true, ApplyPositionFalloff=false)
   &SolveForcesAndVelocity(Clamp Velocity=false, Limit Acceleration=false)
   &ScaleColor(
    Scale Alpha=&FloatFromCurve(FloatCurve=DataInterface() {propertyValues {Curve {keys [
         {arriveTangent=-2.6644, interpMode=RCIM_Cubic, leaveTangent=-2.6644, tangentMode=RCTM_User, time=0.5, value=1.1}
         {interpMode=RCIM_Cubic, tangentMode=RCTM_User, time=1, value=0}
        ]}}})
    ScaleA=true
    ScaleRGB=true
   )
   &ScaleSpriteSize(Uniform Curve Sprite Scale=DataInterface() {propertyValues {Curve {keys [
        {interpMode=RCIM_Cubic, leaveTangent=1.3665, tangentMode=RCTM_Break, time=0, value=0.25}
        {arriveTangent=0.1218, interpMode=RCIM_Cubic, leaveTangent=0.1218, tangentMode=RCTM_User, time=0.9983, value=1}
       ]}}})
  ]
  rendererClasses [&NiagaraSpriteRendererProperties]
  renderers [
   {rendererClass=&NiagaraSpriteRendererProperties}
   {rendererClass=&NiagaraSpriteRendererProperties, rendererIndex=1}
  ]
 }
 WindstreaksJet {
  emitterUpdateScript [
   &EmitterState()
   &SpawnBurst_Instantaneous(Spawn Count=3, Spawn Time=0.05, Use Spawn Probability=false, UseLoopCountLimit=false)
  ]
  particleSpawnScript [
   &InitializeParticle(
    Color=<rgba 0.6517 0.8542 0.3813 0.2>
    Lifetime=0.2
    Sprite Size=&UniformRangedVector2D(Minimum=<7 30>, Maximum=<10 30>)
    Sprite Size Mode=::NewEnumerator4
    UsePositionOffset=false
    Write Material Random=true
   )
   &SkeletalMeshLocation(
    Bone / Socket Index=&SelectIntFromArray(
     Direct Array Index=&ReturnExecIndex()
     Int Selection Array=DataInterface() {propertyValues {IntData [0,1,2]}}
    )
    Bone Sampling Mode=::NewEnumerator4
    MaskSpawnGroup=false
    Skeletal Mesh=$System.Skeletal Mesh
   )
   &AddVelocity(
    Cone Angle=25
    Cone Axis=<1 0 -0.2>
    Module.UseConeFalloff=false
    Module.UseVelDistribution=false
    Override Seed=false
    Velocity Mode=::NewEnumerator2
    Velocity Speed=&Multiply_Float(A=&RandomRangeFloat(Minimum=250, Maximum=300), B=1)
   )
   &ApplyInitialForces()
   &SubUVAnimation(
    Evaluation Type=::NewEnumerator0
    "Sprite Renderer "=DataInterface() {propertyValues {SpriteRenderer=&self:WindstreaksJet_0.NiagaraSpriteRendererProperties_0}}
    SubUV Animation Mode:ENiagara_SubUVLookupModeV2=::NewEnumerator2
    UseEndFrame=false
    UseStartFrame=false
   )
  ]
  particleUpdateScript [
   &ParticleState()
   &ScaleSpriteSize(Scale Factor=&Vector2DFromCurve(
     Vector2Curve=DataInterface() {propertyValues {
       XCurve {keys [
         {arriveTangent=2.5428, interpMode=RCIM_Cubic, leaveTangent=2.5428, tangentMode=RCTM_User, time=0, value=1}
         {arriveTangent=0.1281, interpMode=RCIM_Cubic, leaveTangent=0.1281, tangentMode=RCTM_User, time=1, value=2}
        ]}
       YCurve {keys [
         {arriveTangent=2.8823, interpMode=RCIM_Cubic, leaveTangent=2.8823, tangentMode=RCTM_User, time=0, value=1}
         {arriveTangent=2.0297, interpMode=RCIM_Cubic, leaveTangent=2.0297, tangentMode=RCTM_User, time=1, value=2}
        ]}
      }}
     ScaleCurve=&Multiply_Vector2D_ByFloat(
      Float=&RandomRangeFloat(Minimum=1, Maximum=2)
      Vector2D=&RandomRangeVector2D(Minimum=<0.6 0.75>, Maximum=<0.8 1>)
     )
    ))
   &DynamicMaterialParameters(
    Index 0 Param 1=20
    Index 0 Param 2=1
    Index 0 Param 3=1
    Index 0 Param 4=&FloatFromCurve(FloatCurve=DataInterface() {propertyValues {Curve {keys [{tangentMode=RCTM_User, time=0.1, value=0}, {tangentMode=RCTM_User, time=1, value=0.5}]}}})
    Module.Param0WriteEnabled=true
    Module.Param1WriteEnabled=true
    Module.Param2WriteEnabled=true
    Module.Param3WriteEnabled=true
   ) {module=DynamicMaterialParameters001}
   &Drag(Drag=5, Use Linear Drag=true, Use Rotational Drag=true)
   &InheritSourceMovement(
    Applied Position Scale=&Multiply_VectorByFloat(
     Vector=<1 1 1>
     Float=&FloatFromCurve(FloatCurve=DataInterface() {propertyValues {Curve {keys [<at 0.5 1>,<at 1 0>]}}})
    )
    ApplyForce=false
    ApplyPosition=true
    ApplyPositionFalloff=false
   )
   &SolveForcesAndVelocity(Clamp Velocity=false, Limit Acceleration=false)
   &ScaleColor(
    Scale Alpha=&FloatFromCurve(
     FloatCurve=DataInterface() {propertyValues {Curve {keys [
         {arriveTangent=9.0725, interpMode=RCIM_Cubic, leaveTangent=9.0725, tangentMode=RCTM_User, time=0, value=0}
         {arriveTangent=-2.5249, interpMode=RCIM_Cubic, leaveTangent=-2.5249, tangentMode=RCTM_User, time=0.1408, value=0.9847}
         {arriveTangent=-0.0405, interpMode=RCIM_Cubic, leaveTangent=-0.0405, tangentMode=RCTM_User, time=1, value=0}
        ]}}}
     Scale Curve=&Multiply_Float(
      A=$System.TODAlpha
      B=&SelectFloatFromArray(Direct Array Index=$Engine.QualityLevel, Float Selection Array=DataInterface() {propertyValues {FloatData [0.5,0.5,1,1,1]}})
     )
    )
    Scale RGB=&MakeVectorFromLinearColorRGB(Color=$System.TODColor)
    ScaleA=true
    ScaleRGB=true
   )
  ]
  rendererClasses [&NiagaraSpriteRendererProperties]
  renderers [{rendererClass=&NiagaraSpriteRendererProperties}]
 }
}
summary {
 systemName=NS_Emote_CubicVice_Splat
 emitters [
  {emitterName=WaterDrops, rendererClasses [&NiagaraSpriteRendererProperties]}
  {emitterName=WaterSplash, rendererClasses [&NiagaraSpriteRendererProperties]}
  {emitterName=WindstreaksJet, rendererClasses [&NiagaraSpriteRendererProperties]}
 ]
 userVariables [
  User.bIsFrontEnd {type=NiagaraBool}
  User.bIsFrontEndPreview {type=NiagaraBool}
 ]
}
